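# Anim Blueprint for the HarpTar backpack, offsetting the pack, aiming it at the wearer's pelvis and simulating its zipper and straps.
# Transforms, zipper AnimDynamics and Look At build a cached pose feeding two differently tuned five-strap AnimDynamics chains, selected by front-end state.
# The event graph tracks body mesh, pelvis location and body flags.
# It also sets skydiving gravity and random-force variables, which no AnimGraph node visibly reads.
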AnimBlueprint /BRCosmetics/Accessories/FORT_Backpacks/Backpack_F_MED_HarpTar/Meshes/F_MED_HarpTar_Pack_AnimBP
target_skeleton=&/F_MED_HarpTar_Pack_Skeleton
graphs [
 AnimGraph {nodes [
   n1 = Anim Dynamics - Bone: dyn_strap_a_01!(ComponentPose=Local To Component(LocalPose=Use cached pose 'DefaultPose'()), Alpha=1, GravityOverride:Vector="0,0,50", ExternalForce="0,0,0") -> (Pose) {
    angularDampingOverride=0.9999
    angularSpringConstant=10000
    bAngularSpring=true
    bOverrideAngularDamping=true
    bUseGravityOverride=true
    gravityScale=0.35
    simulationSpace=RootRelative
    alphaBoolBlend {blendOption=Linear}
    boundBone {boneName=dyn_strap_a_01}
    componentAppliedLinearAccClamp=<10000 10000 10000>
    componentLinearAccScale=<0.4 0.4 0.4>
    componentLinearVelScale=<0.4 0.4 0.4>
    gravityOverride {z=50}
    physicsBodyDefinitions [{
      boundBone {boneName=dyn_strap_a_01}
      boxExtents=<10 40 40>
      constraintSetup {angularLimitsMax=<0 15 5>, angularLimitsMin=<0 -15 -40>, angularTarget=<1 0 0>}
      localJointOffset=<3 0 0>
     }]
   }
   n2 = Anim Dynamics - Bone: dyn_strap_a_02!(ComponentPose=@n1, Alpha=1, GravityOverride:Vector="0,0,50", ExternalForce="0,0,0") -> (Pose) {
    angularDampingOverride=0.9999
    angularSpringConstant=10000
    bAngularSpring=true
    bOverrideAngularDamping=true
    bUseGravityOverride=true
    gravityScale=0.35
    simulationSpace=RootRelative
    alphaBoolBlend {blendOption=Linear}
    boundBone {boneName=dyn_strap_a_02}
    componentAppliedLinearAccClamp=<10000 10000 10000>
    componentLinearAccScale=<0.4 0.4 0.4>
    componentLinearVelScale=<0.4 0.4 0.4>
    gravityOverride {z=50}
    physicsBodyDefinitions [{
      boundBone {boneName=dyn_strap_a_02}
      boxExtents=<10 40 40>
      constraintSetup {angularLimitsMax=<0 15 5>, angularLimitsMin=<0 -15 -40>, angularTarget=<1 0 0>}
      localJointOffset=<4 0 0>
     }]
   }
   n3 = Anim Dynamics - Bone: dyn_strap_a_03!(ComponentPose=@n2, Alpha=1, GravityOverride:Vector="0,0,50", ExternalForce="0,0,0") -> (Pose) {
    angularDampingOverride=0.9999
    angularSpringConstant=10000
    bAngularSpring=true
    bOverrideAngularDamping=true
    bUseGravityOverride=true
    gravityScale=0.35
    simulationSpace=RootRelative
    alphaBoolBlend {blendOption=Linear}
    boundBone {boneName=dyn_strap_a_03}
    componentAppliedLinearAccClamp=<10000 10000 10000>
    componentLinearAccScale=<0.4 0.4 0.4>
    componentLinearVelScale=<0.4 0.4 0.4>
    gravityOverride {z=50}
    physicsBodyDefinitions [{
      boundBone {boneName=dyn_strap_a_03}
      boxExtents=<10 40 40>
      constraintSetup {angularLimitsMax=<0 15 5>, angularLimitsMin=<0 -15 -40>, angularTarget=<1 0 0>}
      localJointOffset=<5 0 0>
     }]
   }
   n4 = Anim Dynamics - Bone: dyn_strap_a_04!(ComponentPose=@n3, Alpha=1, GravityOverride:Vector="0,0,50", ExternalForce="0,0,0") -> (Pose) {
    angularDampingOverride=0.9999
    angularSpringConstant=10000
    bAngularSpring=true
    bOverrideAngularDamping=true
    bUseGravityOverride=true
    gravityScale=0.35
    simulationSpace=RootRelative
    alphaBoolBlend {blendOption=Linear}
    boundBone {boneName=dyn_strap_a_04}
    componentAppliedLinearAccClamp=<10000 10000 10000>
    componentLinearAccScale=<0.4 0.4 0.4>
    componentLinearVelScale=<0.4 0.4 0.4>
    gravityOverride {z=50}
    physicsBodyDefinitions [{
      boundBone {boneName=dyn_strap_a_04}
      boxExtents=<10 40 40>
      constraintSetup {angularLimitsMax=<0 15 5>, angularLimitsMin=<0 -15 -40>, angularTarget=<1 0 0>}
      localJointOffset=<5 0 0>
     }]
   }
   n5 = Anim Dynamics - Bone: dyn_strap_b_01!(ComponentPose=@n4, Alpha=1, GravityOverride:Vector="0,0,50", ExternalForce="0,0,0") -> (Pose) {
    angularDampingOverride=0.9999
    angularSpringConstant=10000
    bAngularSpring=true
    bOverrideAngularDamping=true
    bUseGravityOverride=true
    gravityScale=0.35
    simulationSpace=RootRelative
    alphaBoolBlend {blendOption=Linear}
    boundBone {boneName=dyn_strap_b_01}
    componentAppliedLinearAccClamp=<10000 10000 10000>
    componentLinearAccScale=<0.4 0.4 0.4>
    componentLinearVelScale=<0.4 0.4 0.4>
    gravityOverride {z=50}
    physicsBodyDefinitions [{
      boundBone {boneName=dyn_strap_b_01}
      boxExtents=<10 40 40>
      constraintSetup {angularLimitsMax=<0 15 5>, angularLimitsMin=<0 -15 -40>, angularTarget=<1 0 0>}
      localJointOffset=<5 0 0>
     }]
   }
   n6 = Anim Dynamics - Bone: dyn_strap_b_02!(ComponentPose=@n5, Alpha=1, GravityOverride:Vector="0,0,50", ExternalForce="0,0,0") -> (Pose) {
    angularDampingOverride=0.9999
    angularSpringConstant=10000
    bAngularSpring=true
    bOverrideAngularDamping=true
    bUseGravityOverride=true
    gravityScale=0.35
    simulationSpace=RootRelative
    alphaBoolBlend {blendOption=Linear}
    boundBone {boneName=dyn_strap_b_02}
    componentAppliedLinearAccClamp=<10000 10000 10000>
    componentLinearAccScale=<0.4 0.4 0.4>
    componentLinearVelScale=<0.4 0.4 0.4>
    gravityOverride {z=50}
    physicsBodyDefinitions [{
      boundBone {boneName=dyn_strap_b_02}
      boxExtents=<10 40 40>
      constraintSetup {angularLimitsMax=<0 15 5>, angularLimitsMin=<0 -15 -40>, angularTarget=<1 0 0>}
      localJointOffset=<5 0 0>
     }]
   }
   n7 = Anim Dynamics - Bone: dyn_strap_b_03!(ComponentPose=@n6, Alpha=1, GravityOverride:Vector="0,0,50", ExternalForce="0,0,0") -> (Pose) {
    angularDampingOverride=0.9999
    angularSpringConstant=10000
    bAngularSpring=true
    bOverrideAngularDamping=true
    bUseGravityOverride=true
    gravityScale=0.35
    simulationSpace=RootRelative
    alphaBoolBlend {blendOption=Linear}
    boundBone {boneName=dyn_strap_b_03}
    componentAppliedLinearAccClamp=<10000 10000 10000>
    componentLinearAccScale=<0.4 0.4 0.4>
    componentLinearVelScale=<0.4 0.4 0.4>
    gravityOverride {z=50}
    physicsBodyDefinitions [{
      boundBone {boneName=dyn_strap_b_03}
      boxExtents=<10 40 40>
      constraintSetup {angularLimitsMax=<0 15 5>, angularLimitsMin=<0 -15 -40>, angularTarget=<1 0 0>}
      localJointOffset=<3 0 0>
     }]
   }
   n8 = Anim Dynamics - Bone: dyn_strap_b_04!(ComponentPose=@n7, Alpha=1, GravityOverride:Vector="0,0,50", ExternalForce="0,0,0") -> (Pose) {
    angularDampingOverride=0.9999
    angularSpringConstant=10000
    bAngularSpring=true
    bOverrideAngularDamping=true
    bUseGravityOverride=true
    gravityScale=0.35
    simulationSpace=RootRelative
    alphaBoolBlend {blendOption=Linear}
    boundBone {boneName=dyn_strap_b_04}
    componentAppliedLinearAccClamp=<10000 10000 10000>
    componentLinearAccScale=<0.4 0.4 0.4>
    componentLinearVelScale=<0.4 0.4 0.4>
    gravityOverride {z=50}
    physicsBodyDefinitions [{
      boundBone {boneName=dyn_strap_b_04}
      boxExtents=<10 40 40>
      constraintSetup {angularLimitsMax=<0 15 5>, angularLimitsMin=<0 -15 -40>, angularTarget=<1 0 0>}
      localJointOffset=<4 0 0>
     }]
   }
   n9 = Anim Dynamics - Bone: dyn_strap_c_01!(ComponentPose=@n8, Alpha=1, GravityOverride:Vector="0,0,50", ExternalForce="0,0,0") -> (Pose) {
    angularDampingOverride=0.9999
    angularSpringConstant=10000
    bAngularSpring=true
    bOverrideAngularDamping=true
    bUseGravityOverride=true
    gravityScale=0.35
    simulationSpace=RootRelative
    alphaBoolBlend {blendOption=Linear}
    boundBone {boneName=dyn_strap_c_01}
    componentAppliedLinearAccClamp=<10000 10000 10000>
    componentLinearAccScale=<0.4 0.4 0.4>
    componentLinearVelScale=<0.4 0.4 0.4>
    gravityOverride {z=50}
    physicsBodyDefinitions [{
      boundBone {boneName=dyn_strap_c_01}
      boxExtents=<10 40 40>
      constraintSetup {angularLimitsMax=<0 15 5>, angularLimitsMin=<0 -15 -40>, angularTarget=<1 0 0>}
      localJointOffset=<3 0 0>
     }]
   }
   n10 = Anim Dynamics - Bone: dyn_strap_c_02!(ComponentPose=@n9, Alpha=1, GravityOverride:Vector="0,0,50", ExternalForce="0,0,0") -> (Pose) {
    angularDampingOverride=0.9999
    angularSpringConstant=10000
    bAngularSpring=true
    bOverrideAngularDamping=true
    bUseGravityOverride=true
    gravityScale=0.35
    simulationSpace=RootRelative
    alphaBoolBlend {blendOption=Linear}
    boundBone {boneName=dyn_strap_c_02}
    componentAppliedLinearAccClamp=<10000 10000 10000>
    componentLinearAccScale=<0.4 0.4 0.4>
    componentLinearVelScale=<0.4 0.4 0.4>
    gravityOverride {z=50}
    physicsBodyDefinitions [{
      boundBone {boneName=dyn_strap_c_02}
      boxExtents=<10 40 40>
      constraintSetup {angularLimitsMax=<0 15 5>, angularLimitsMin=<0 -15 -40>, angularTarget=<1 0 0>}
      localJointOffset=<4 0 0>
     }]
   }
   n11 = Anim Dynamics - Bone: dyn_strap_c_03!(ComponentPose=@n10, Alpha=1, GravityOverride:Vector="0,0,50", ExternalForce="0,0,0") -> (Pose) {
    angularDampingOverride=0.9999
    angularSpringConstant=10000
    bAngularSpring=true
    bOverrideAngularDamping=true
    bUseGravityOverride=true
    gravityScale=0.35
    simulationSpace=RootRelative
    alphaBoolBlend {blendOption=Linear}
    boundBone {boneName=dyn_strap_c_03}
    componentAppliedLinearAccClamp=<10000 10000 10000>
    componentLinearAccScale=<0.4 0.4 0.4>
    componentLinearVelScale=<0.4 0.4 0.4>
    gravityOverride {z=50}
    physicsBodyDefinitions [{
      boundBone {boneName=dyn_strap_c_03}
      boxExtents=<10 40 40>
      constraintSetup {angularLimitsMax=<0 15 5>, angularLimitsMin=<0 -15 -40>, angularTarget=<1 0 0>}
      localJointOffset=<5 0 0>
     }]
   }
   n12 = Anim Dynamics - Bone: dyn_strap_c_04!(ComponentPose=@n11, Alpha=1, GravityOverride:Vector="0,0,50", ExternalForce="0,0,0") -> (Pose) {
    angularDampingOverride=0.9999
    angularSpringConstant=10000
    bAngularSpring=true
    bOverrideAngularDamping=true
    bUseGravityOverride=true
    gravityScale=0.35
    simulationSpace=RootRelative
    alphaBoolBlend {blendOption=Linear}
    boundBone {boneName=dyn_strap_c_04}
    componentAppliedLinearAccClamp=<10000 10000 10000>
    componentLinearAccScale=<0.4 0.4 0.4>
    componentLinearVelScale=<0.4 0.4 0.4>
    gravityOverride {z=50}
    physicsBodyDefinitions [{
      boundBone {boneName=dyn_strap_c_04}
      boxExtents=<10 40 40>
      constraintSetup {angularLimitsMax=<0 15 5>, angularLimitsMin=<0 -15 -40>, angularTarget=<1 0 0>}
      localJointOffset=<5 0 0>
     }]
   }
   n13 = Anim Dynamics - Bone: dyn_strap_d_01!(ComponentPose=@n12, Alpha=1, GravityOverride:Vector="0,0,50", ExternalForce="0,0,0") -> (Pose) {
    angularDampingOverride=0.9999
    angularSpringConstant=10000
    bAngularSpring=true
    bOverrideAngularDamping=true
    bUseGravityOverride=true
    gravityScale=0.35
    simulationSpace=RootRelative
    alphaBoolBlend {blendOption=Linear}
    boundBone {boneName=dyn_strap_d_01}
    componentAppliedLinearAccClamp=<10000 10000 10000>
    componentLinearAccScale=<0.4 0.4 0.4>
    componentLinearVelScale=<0.4 0.4 0.4>
    gravityOverride {z=50}
    physicsBodyDefinitions [{
      boundBone {boneName=dyn_strap_d_01}
      boxExtents=<10 40 40>
      constraintSetup {angularLimitsMax=<0 10 5>, angularLimitsMin=<0 -10 -40>, angularTarget=<1 0 0>}
      localJointOffset=<4 0 0>
     }]
   }
   n14 = Anim Dynamics - Bone: dyn_strap_d_02!(ComponentPose=@n13, Alpha=1, GravityOverride:Vector="0,0,50", ExternalForce="0,0,0") -> (Pose) {
    angularDampingOverride=0.9999
    angularSpringConstant=10000
    bAngularSpring=true
    bOverrideAngularDamping=true
    bUseGravityOverride=true
    gravityScale=0.35
    simulationSpace=RootRelative
    alphaBoolBlend {blendOption=Linear}
    boundBone {boneName=dyn_strap_d_02}
    componentAppliedLinearAccClamp=<10000 10000 10000>
    componentLinearAccScale=<0.4 0.4 0.4>
    componentLinearVelScale=<0.4 0.4 0.4>
    gravityOverride {z=50}
    physicsBodyDefinitions [{
      boundBone {boneName=dyn_strap_d_02}
      boxExtents=<10 40 40>
      constraintSetup {angularLimitsMax=<0 10 5>, angularLimitsMin=<0 -10 -40>, angularTarget=<1 0 0>}
      localJointOffset=<4 0 0>
     }]
   }
   n15 = Anim Dynamics - Bone: dyn_strap_d_03!(ComponentPose=@n14, Alpha=1, GravityOverride:Vector="0,0,50", ExternalForce="0,0,0") -> (Pose) {
    angularDampingOverride=0.9999
    angularSpringConstant=10000
    bAngularSpring=true
    bOverrideAngularDamping=true
    bUseGravityOverride=true
    gravityScale=0.35
    simulationSpace=RootRelative
    alphaBoolBlend {blendOption=Linear}
    boundBone {boneName=dyn_strap_d_03}
    componentAppliedLinearAccClamp=<10000 10000 10000>
    componentLinearAccScale=<0.4 0.4 0.4>
    componentLinearVelScale=<0.4 0.4 0.4>
    gravityOverride {z=50}
    physicsBodyDefinitions [{
      boundBone {boneName=dyn_strap_d_03}
      boxExtents=<10 40 40>
      constraintSetup {angularLimitsMax=<0 10 5>, angularLimitsMin=<0 -10 -40>, angularTarget=<1 0 0>}
      localJointOffset=<3 0 0>
     }]
   }
   n16 = Anim Dynamics - Bone: dyn_strap_d_04!(ComponentPose=@n15, Alpha=1, GravityOverride:Vector="0,0,50", ExternalForce="0,0,0") -> (Pose) {
    angularDampingOverride=0.9999
    angularSpringConstant=10000
    bAngularSpring=true
    bOverrideAngularDamping=true
    bUseGravityOverride=true
    gravityScale=0.35
    simulationSpace=RootRelative
    alphaBoolBlend {blendOption=Linear}
    boundBone {boneName=dyn_strap_d_04}
    componentAppliedLinearAccClamp=<10000 10000 10000>
    componentLinearAccScale=<0.4 0.4 0.4>
    componentLinearVelScale=<0.4 0.4 0.4>
    gravityOverride {z=50}
    physicsBodyDefinitions [{
      boundBone {boneName=dyn_strap_d_04}
      boxExtents=<10 40 40>
      constraintSetup {angularLimitsMax=<0 10 5>, angularLimitsMin=<0 -10 -40>, angularTarget=<1 0 0>}
      localJointOffset=<4 0 0>
     }]
   }
   n17 = Anim Dynamics - Bone: dyn_strap_e_01!(ComponentPose=@n16, Alpha=1, GravityOverride:Vector="0,0,50", ExternalForce="0,0,0") -> (Pose) {
    angularDampingOverride=0.9999
    angularSpringConstant=10000
    bAngularSpring=true
    bOverrideAngularDamping=true
    bUseGravityOverride=true
    gravityScale=0.35
    simulationSpace=RootRelative
    alphaBoolBlend {blendOption=Linear}
    boundBone {boneName=dyn_strap_e_01}
    componentAppliedLinearAccClamp=<10000 10000 10000>
    componentLinearAccScale=<0.4 0.4 0.4>
    componentLinearVelScale=<0.4 0.4 0.4>
    gravityOverride {z=50}
    physicsBodyDefinitions [{
      boundBone {boneName=dyn_strap_e_01}
      boxExtents=<10 40 40>
      constraintSetup {angularLimitsMax=<0 10 5>, angularLimitsMin=<0 -15 -40>, angularTarget=<1 0 0>}
      localJointOffset=<5 0 0>
     }]
   }
   n18 = Anim Dynamics - Bone: dyn_strap_e_02!(ComponentPose=@n17, Alpha=1, GravityOverride:Vector="0,0,50", ExternalForce="0,0,0") -> (Pose) {
    angularDampingOverride=0.9999
    angularSpringConstant=10000
    bAngularSpring=true
    bOverrideAngularDamping=true
    bUseGravityOverride=true
    gravityScale=0.35
    simulationSpace=RootRelative
    alphaBoolBlend {blendOption=Linear}
    boundBone {boneName=dyn_strap_e_02}
    componentAppliedLinearAccClamp=<10000 10000 10000>
    componentLinearAccScale=<0.4 0.4 0.4>
    componentLinearVelScale=<0.4 0.4 0.4>
    gravityOverride {z=50}
    physicsBodyDefinitions [{
      boundBone {boneName=dyn_strap_e_02}
      boxExtents=<10 40 40>
      constraintSetup {angularLimitsMax=<0 10 5>, angularLimitsMin=<0 -15 -40>, angularTarget=<1 0 0>}
      localJointOffset=<3 0 0>
     }]
   }
   n19 = Anim Dynamics - Bone: dyn_strap_e_03!(ComponentPose=@n18, Alpha=1, GravityOverride:Vector="0,0,50", ExternalForce="0,0,0") -> (Pose) {
    angularDampingOverride=0.9999
    angularSpringConstant=10000
    bAngularSpring=true
    bOverrideAngularDamping=true
    bUseGravityOverride=true
    gravityScale=0.35
    simulationSpace=RootRelative
    alphaBoolBlend {blendOption=Linear}
    boundBone {boneName=dyn_strap_e_03}
    componentAppliedLinearAccClamp=<10000 10000 10000>
    componentLinearAccScale=<0.4 0.4 0.4>
    componentLinearVelScale=<0.4 0.4 0.4>
    gravityOverride {z=50}
    physicsBodyDefinitions [{
      boundBone {boneName=dyn_strap_e_03}
      boxExtents=<10 40 40>
      constraintSetup {angularLimitsMax=<0 10 5>, angularLimitsMin=<0 -15 -40>, angularTarget=<1 0 0>}
      localJointOffset=<2 0 0>
     }]
   }
   n20 = Anim Dynamics - Bone: dyn_strap_e_04!(ComponentPose=@n19, Alpha=1, GravityOverride:Vector="0,0,50", ExternalForce="0,0,0") -> (Pose) {
    angularDampingOverride=0.9999
    angularSpringConstant=10000
    bAngularSpring=true
    bOverrideAngularDamping=true
    bUseGravityOverride=true
    gravityScale=0.35
    simulationSpace=RootRelative
    alphaBoolBlend {blendOption=Linear}
    boundBone {boneName=dyn_strap_e_04}
    componentAppliedLinearAccClamp=<10000 10000 10000>
    componentLinearAccScale=<0.4 0.4 0.4>
    componentLinearVelScale=<0.4 0.4 0.4>
    gravityOverride {z=50}
    physicsBodyDefinitions [{
      boundBone {boneName=dyn_strap_e_04}
      boxExtents=<10 40 40>
      constraintSetup {angularLimitsMax=<0 10 5>, angularLimitsMin=<0 -15 -40>, angularTarget=<1 0 0>}
      localJointOffset=<4 0 0>
     }]
   }
   n21 = Anim Dynamics - Bone: dyn_strap_a_01(ComponentPose=Local To Component(LocalPose=Use cached pose 'DefaultPose'()), Alpha=1, GravityOverride="0,0,50", ExternalForce="0,0,0") {
    angularDampingOverride=1
    angularSpringConstant=10000
    bAngularSpring=true
    bOverrideAngularDamping=true
    bUseGravityOverride=true
    gravityScale=0.35
    simulationSpace=RootRelative
    alphaBoolBlend {blendOption=Linear}
    boundBone {boneName=dyn_strap_a_01}
    componentAppliedLinearAccClamp=<2000 2000 2500>
    componentLinearAccScale=<0.6 0.6 0.6>
    componentLinearVelScale=<0.6 0.6 0.6>
    gravityOverride {z=50}
    physicsBodyDefinitions [{
      boundBone {boneName=dyn_strap_a_01}
      boxExtents=<10 40 40>
      constraintSetup {angularLimitsMax=<0 15 5>, angularLimitsMin=<0 -15 -40>, angularTarget=<1 0 0>}
      localJointOffset=<3 0 0>
     }]
   }
   n22 = Anim Dynamics - Bone: dyn_strap_a_02(ComponentPose=@n21, Alpha=1, GravityOverride="0,0,50", ExternalForce="0,0,0") {
    angularDampingOverride=1
    angularSpringConstant=10000
    bAngularSpring=true
    bOverrideAngularDamping=true
    bUseGravityOverride=true
    gravityScale=0.35
    simulationSpace=RootRelative
    alphaBoolBlend {blendOption=Linear}
    boundBone {boneName=dyn_strap_a_02}
    componentAppliedLinearAccClamp=<2000 2000 2500>
    componentLinearAccScale=<0.6 0.6 0.6>
    componentLinearVelScale=<0.6 0.6 0.6>
    gravityOverride {z=50}
    physicsBodyDefinitions [{
      boundBone {boneName=dyn_strap_a_02}
      boxExtents=<10 40 40>
      constraintSetup {angularLimitsMax=<0 15 5>, angularLimitsMin=<0 -15 -40>, angularTarget=<1 0 0>}
      localJointOffset=<4 0 0>
     }]
   }
   n23 = Anim Dynamics - Bone: dyn_strap_a_03(ComponentPose=@n22, Alpha=1, GravityOverride="0,0,50", ExternalForce="0,0,0") {
    angularDampingOverride=1
    angularSpringConstant=10000
    bAngularSpring=true
    bOverrideAngularDamping=true
    bUseGravityOverride=true
    gravityScale=0.35
    simulationSpace=RootRelative
    alphaBoolBlend {blendOption=Linear}
    boundBone {boneName=dyn_strap_a_03}
    componentAppliedLinearAccClamp=<2000 2000 2500>
    componentLinearAccScale=<0.6 0.6 0.6>
    componentLinearVelScale=<0.6 0.6 0.6>
    gravityOverride {z=50}
    physicsBodyDefinitions [{
      boundBone {boneName=dyn_strap_a_03}
      boxExtents=<10 40 40>
      constraintSetup {angularLimitsMax=<0 15 5>, angularLimitsMin=<0 -15 -40>, angularTarget=<1 0 0>}
      localJointOffset=<5 0 0>
     }]
   }
   n24 = Anim Dynamics - Bone: dyn_strap_a_04(ComponentPose=@n23, Alpha=1, GravityOverride="0,0,50", ExternalForce="0,0,0") {
    angularDampingOverride=1
    angularSpringConstant=10000
    bAngularSpring=true
    bOverrideAngularDamping=true
    bUseGravityOverride=true
    gravityScale=0.35
    simulationSpace=RootRelative
    alphaBoolBlend {blendOption=Linear}
    boundBone {boneName=dyn_strap_a_04}
    componentAppliedLinearAccClamp=<2000 2000 2500>
    componentLinearAccScale=<0.6 0.6 0.6>
    componentLinearVelScale=<0.6 0.6 0.6>
    gravityOverride {z=50}
    physicsBodyDefinitions [{
      boundBone {boneName=dyn_strap_a_04}
      boxExtents=<10 40 40>
      constraintSetup {angularLimitsMax=<0 15 5>, angularLimitsMin=<0 -15 -40>, angularTarget=<1 0 0>}
      localJointOffset=<5 0 0>
     }]
   }
   n25 = Anim Dynamics - Bone: dyn_strap_b_01(ComponentPose=@n24, Alpha=1, GravityOverride="0,0,50", ExternalForce="0,0,0") {
    angularDampingOverride=1
    angularSpringConstant=10000
    bAngularSpring=true
    bOverrideAngularDamping=true
    bUseGravityOverride=true
    gravityScale=0.35
    simulationSpace=RootRelative
    alphaBoolBlend {blendOption=Linear}
    boundBone {boneName=dyn_strap_b_01}
    componentAppliedLinearAccClamp=<2000 2000 2500>
    componentLinearAccScale=<0.6 0.6 0.6>
    componentLinearVelScale=<0.6 0.6 0.6>
    gravityOverride {z=50}
    physicsBodyDefinitions [{
      boundBone {boneName=dyn_strap_b_01}
      boxExtents=<10 40 40>
      constraintSetup {angularLimitsMax=<0 15 5>, angularLimitsMin=<0 -15 -40>, angularTarget=<1 0 0>}
      localJointOffset=<5 0 0>
     }]
   }
   n26 = Anim Dynamics - Bone: dyn_strap_b_02(ComponentPose=@n25, Alpha=1, GravityOverride="0,0,50", ExternalForce="0,0,0") {
    angularDampingOverride=1
    angularSpringConstant=10000
    bAngularSpring=true
    bOverrideAngularDamping=true
    bUseGravityOverride=true
    gravityScale=0.35
    simulationSpace=RootRelative
    alphaBoolBlend {blendOption=Linear}
    boundBone {boneName=dyn_strap_b_02}
    componentAppliedLinearAccClamp=<2000 2000 2500>
    componentLinearAccScale=<0.6 0.6 0.6>
    componentLinearVelScale=<0.6 0.6 0.6>
    gravityOverride {z=50}
    physicsBodyDefinitions [{
      boundBone {boneName=dyn_strap_b_02}
      boxExtents=<10 40 40>
      constraintSetup {angularLimitsMax=<0 15 5>, angularLimitsMin=<0 -15 -40>, angularTarget=<1 0 0>}
      localJointOffset=<5 0 0>
     }]
   }
   n27 = Anim Dynamics - Bone: dyn_strap_b_03(ComponentPose=@n26, Alpha=1, GravityOverride="0,0,50", ExternalForce="0,0,0") {
    angularDampingOverride=1
    angularSpringConstant=10000
    bAngularSpring=true
    bOverrideAngularDamping=true
    bUseGravityOverride=true
    gravityScale=0.35
    simulationSpace=RootRelative
    alphaBoolBlend {blendOption=Linear}
    boundBone {boneName=dyn_strap_b_03}
    componentAppliedLinearAccClamp=<2000 2000 2500>
    componentLinearAccScale=<0.6 0.6 0.6>
    componentLinearVelScale=<0.6 0.6 0.6>
    gravityOverride {z=50}
    physicsBodyDefinitions [{
      boundBone {boneName=dyn_strap_b_03}
      boxExtents=<10 40 40>
      constraintSetup {angularLimitsMax=<0 15 5>, angularLimitsMin=<0 -15 -40>, angularTarget=<1 0 0>}
      localJointOffset=<3 0 0>
     }]
   }
   n28 = Anim Dynamics - Bone: dyn_strap_b_04(ComponentPose=@n27, Alpha=1, GravityOverride="0,0,50", ExternalForce="0,0,0") {
    angularDampingOverride=1
    angularSpringConstant=10000
    bAngularSpring=true
    bOverrideAngularDamping=true
    bUseGravityOverride=true
    gravityScale=0.35
    simulationSpace=RootRelative
    alphaBoolBlend {blendOption=Linear}
    boundBone {boneName=dyn_strap_b_04}
    componentAppliedLinearAccClamp=<2000 2000 2500>
    componentLinearAccScale=<0.6 0.6 0.6>
    componentLinearVelScale=<0.6 0.6 0.6>
    gravityOverride {z=50}
    physicsBodyDefinitions [{
      boundBone {boneName=dyn_strap_b_04}
      boxExtents=<10 40 40>
      constraintSetup {angularLimitsMax=<0 15 5>, angularLimitsMin=<0 -15 -40>, angularTarget=<1 0 0>}
      localJointOffset=<4 0 0>
     }]
   }
   n29 = Anim Dynamics - Bone: dyn_strap_c_01(ComponentPose=@n28, Alpha=1, GravityOverride="0,0,50", ExternalForce="0,0,0") {
    angularDampingOverride=1
    angularSpringConstant=10000
    bAngularSpring=true
    bOverrideAngularDamping=true
    bUseGravityOverride=true
    gravityScale=0.35
    simulationSpace=RootRelative
    alphaBoolBlend {blendOption=Linear}
    boundBone {boneName=dyn_strap_c_01}
    componentAppliedLinearAccClamp=<2000 2000 2500>
    componentLinearAccScale=<0.6 0.6 0.6>
    componentLinearVelScale=<0.6 0.6 0.6>
    gravityOverride {z=50}
    physicsBodyDefinitions [{
      boundBone {boneName=dyn_strap_c_01}
      boxExtents=<10 40 40>
      constraintSetup {angularLimitsMax=<0 15 5>, angularLimitsMin=<0 -15 -40>, angularTarget=<1 0 0>}
      localJointOffset=<3 0 0>
     }]
   }
   n30 = Anim Dynamics - Bone: dyn_strap_c_02(ComponentPose=@n29, Alpha=1, GravityOverride="0,0,50", ExternalForce="0,0,0") {
    angularDampingOverride=1
    angularSpringConstant=10000
    bAngularSpring=true
    bOverrideAngularDamping=true
    bUseGravityOverride=true
    gravityScale=0.35
    simulationSpace=RootRelative
    alphaBoolBlend {blendOption=Linear}
    boundBone {boneName=dyn_strap_c_02}
    componentAppliedLinearAccClamp=<2000 2000 2500>
    componentLinearAccScale=<0.6 0.6 0.6>
    componentLinearVelScale=<0.6 0.6 0.6>
    gravityOverride {z=50}
    physicsBodyDefinitions [{
      boundBone {boneName=dyn_strap_c_02}
      boxExtents=<10 40 40>
      constraintSetup {angularLimitsMax=<0 15 5>, angularLimitsMin=<0 -15 -40>, angularTarget=<1 0 0>}
      localJointOffset=<4 0 0>
     }]
   }
   n31 = Anim Dynamics - Bone: dyn_strap_c_03(ComponentPose=@n30, Alpha=1, GravityOverride="0,0,50", ExternalForce="0,0,0") {
    angularDampingOverride=1
    angularSpringConstant=10000
    bAngularSpring=true
    bOverrideAngularDamping=true
    bUseGravityOverride=true
    gravityScale=0.35
    simulationSpace=RootRelative
    alphaBoolBlend {blendOption=Linear}
    boundBone {boneName=dyn_strap_c_03}
    componentAppliedLinearAccClamp=<2000 2000 2500>
    componentLinearAccScale=<0.6 0.6 0.6>
    componentLinearVelScale=<0.6 0.6 0.6>
    gravityOverride {z=50}
    physicsBodyDefinitions [{
      boundBone {boneName=dyn_strap_c_03}
      boxExtents=<10 40 40>
      constraintSetup {angularLimitsMax=<0 15 5>, angularLimitsMin=<0 -15 -40>, angularTarget=<1 0 0>}
      localJointOffset=<5 0 0>
     }]
   }
   n32 = Anim Dynamics - Bone: dyn_strap_c_04(ComponentPose=@n31, Alpha=1, GravityOverride="0,0,50", ExternalForce="0,0,0") {
    angularDampingOverride=1
    angularSpringConstant=10000
    bAngularSpring=true
    bOverrideAngularDamping=true
    bUseGravityOverride=true
    gravityScale=0.35
    simulationSpace=RootRelative
    alphaBoolBlend {blendOption=Linear}
    boundBone {boneName=dyn_strap_c_04}
    componentAppliedLinearAccClamp=<2000 2000 2500>
    componentLinearAccScale=<0.6 0.6 0.6>
    componentLinearVelScale=<0.6 0.6 0.6>
    gravityOverride {z=50}
    physicsBodyDefinitions [{
      boundBone {boneName=dyn_strap_c_04}
      boxExtents=<10 40 40>
      constraintSetup {angularLimitsMax=<0 15 5>, angularLimitsMin=<0 -15 -40>, angularTarget=<1 0 0>}
      localJointOffset=<5 0 0>
     }]
   }
   n33 = Anim Dynamics - Bone: dyn_strap_d_01(ComponentPose=@n32, Alpha=1, GravityOverride="0,0,50", ExternalForce="0,0,0") {
    angularDampingOverride=1
    angularSpringConstant=10000
    bAngularSpring=true
    bOverrideAngularDamping=true
    bUseGravityOverride=true
    gravityScale=0.35
    simulationSpace=RootRelative
    alphaBoolBlend {blendOption=Linear}
    boundBone {boneName=dyn_strap_d_01}
    componentAppliedLinearAccClamp=<2000 2000 2500>
    componentLinearAccScale=<0.6 0.6 0.6>
    componentLinearVelScale=<0.6 0.6 0.6>
    gravityOverride {z=50}
    physicsBodyDefinitions [{
      boundBone {boneName=dyn_strap_d_01}
      boxExtents=<10 40 40>
      constraintSetup {angularLimitsMax=<0 10 5>, angularLimitsMin=<0 -10 -40>, angularTarget=<1 0 0>}
      localJointOffset=<4 0 0>
     }]
   }
   n34 = Anim Dynamics - Bone: dyn_strap_d_02(ComponentPose=@n33, Alpha=1, GravityOverride="0,0,50", ExternalForce="0,0,0") {
    angularDampingOverride=1
    angularSpringConstant=10000
    bAngularSpring=true
    bOverrideAngularDamping=true
    bUseGravityOverride=true
    gravityScale=0.35
    simulationSpace=RootRelative
    alphaBoolBlend {blendOption=Linear}
    boundBone {boneName=dyn_strap_d_02}
    componentAppliedLinearAccClamp=<2000 2000 2500>
    componentLinearAccScale=<0.6 0.6 0.6>
    componentLinearVelScale=<0.6 0.6 0.6>
    gravityOverride {z=50}
    physicsBodyDefinitions [{
      boundBone {boneName=dyn_strap_d_02}
      boxExtents=<10 40 40>
      constraintSetup {angularLimitsMax=<0 10 5>, angularLimitsMin=<0 -10 -40>, angularTarget=<1 0 0>}
      localJointOffset=<4 0 0>
     }]
   }
   n35 = Anim Dynamics - Bone: dyn_strap_d_03(ComponentPose=@n34, Alpha=1, GravityOverride="0,0,50", ExternalForce="0,0,0") {
    angularDampingOverride=1
    angularSpringConstant=10000
    bAngularSpring=true
    bOverrideAngularDamping=true
    bUseGravityOverride=true
    gravityScale=0.35
    simulationSpace=RootRelative
    alphaBoolBlend {blendOption=Linear}
    boundBone {boneName=dyn_strap_d_03}
    componentAppliedLinearAccClamp=<2000 2000 2500>
    componentLinearAccScale=<0.6 0.6 0.6>
    componentLinearVelScale=<0.6 0.6 0.6>
    gravityOverride {z=50}
    physicsBodyDefinitions [{
      boundBone {boneName=dyn_strap_d_03}
      boxExtents=<10 40 40>
      constraintSetup {angularLimitsMax=<0 10 5>, angularLimitsMin=<0 -10 -40>, angularTarget=<1 0 0>}
      localJointOffset=<3 0 0>
     }]
   }
   n36 = Anim Dynamics - Bone: dyn_strap_d_04(ComponentPose=@n35, Alpha=1, GravityOverride="0,0,50", ExternalForce="0,0,0") {
    angularDampingOverride=1
    angularSpringConstant=10000
    bAngularSpring=true
    bOverrideAngularDamping=true
    bUseGravityOverride=true
    gravityScale=0.35
    simulationSpace=RootRelative
    alphaBoolBlend {blendOption=Linear}
    boundBone {boneName=dyn_strap_d_04}
    componentAppliedLinearAccClamp=<2000 2000 2500>
    componentLinearAccScale=<0.6 0.6 0.6>
    componentLinearVelScale=<0.6 0.6 0.6>
    gravityOverride {z=50}
    physicsBodyDefinitions [{
      boundBone {boneName=dyn_strap_d_04}
      boxExtents=<10 40 40>
      constraintSetup {angularLimitsMax=<0 10 5>, angularLimitsMin=<0 -10 -40>, angularTarget=<1 0 0>}
      localJointOffset=<4 0 0>
     }]
   }
   n37 = Anim Dynamics - Bone: dyn_strap_e_01(ComponentPose=@n36, Alpha=1, GravityOverride="0,0,50", ExternalForce="0,0,0") {
    angularDampingOverride=1
    angularSpringConstant=10000
    bAngularSpring=true
    bOverrideAngularDamping=true
    bUseGravityOverride=true
    gravityScale=0.35
    simulationSpace=RootRelative
    alphaBoolBlend {blendOption=Linear}
    boundBone {boneName=dyn_strap_e_01}
    componentAppliedLinearAccClamp=<2000 2000 2500>
    componentLinearAccScale=<0.6 0.6 0.6>
    componentLinearVelScale=<0.6 0.6 0.6>
    gravityOverride {z=50}
    physicsBodyDefinitions [{
      boundBone {boneName=dyn_strap_e_01}
      boxExtents=<10 40 40>
      constraintSetup {angularLimitsMax=<0 10 5>, angularLimitsMin=<0 -15 -40>, angularTarget=<1 0 0>}
      localJointOffset=<5 0 0>
     }]
   }
   n38 = Anim Dynamics - Bone: dyn_strap_e_02(ComponentPose=@n37, Alpha=1, GravityOverride="0,0,50", ExternalForce="0,0,0") {
    angularDampingOverride=1
    angularSpringConstant=10000
    bAngularSpring=true
    bOverrideAngularDamping=true
    bUseGravityOverride=true
    gravityScale=0.35
    simulationSpace=RootRelative
    alphaBoolBlend {blendOption=Linear}
    boundBone {boneName=dyn_strap_e_02}
    componentAppliedLinearAccClamp=<2000 2000 2500>
    componentLinearAccScale=<0.6 0.6 0.6>
    componentLinearVelScale=<0.6 0.6 0.6>
    gravityOverride {z=50}
    physicsBodyDefinitions [{
      boundBone {boneName=dyn_strap_e_02}
      boxExtents=<10 40 40>
      constraintSetup {angularLimitsMax=<0 10 5>, angularLimitsMin=<0 -15 -40>, angularTarget=<1 0 0>}
      localJointOffset=<3 0 0>
     }]
   }
   n39 = Anim Dynamics - Bone: dyn_strap_e_03(ComponentPose=@n38, Alpha=1, GravityOverride="0,0,50", ExternalForce="0,0,0") {
    angularDampingOverride=1
    angularSpringConstant=10000
    bAngularSpring=true
    bOverrideAngularDamping=true
    bUseGravityOverride=true
    gravityScale=0.35
    simulationSpace=RootRelative
    alphaBoolBlend {blendOption=Linear}
    boundBone {boneName=dyn_strap_e_03}
    componentAppliedLinearAccClamp=<2000 2000 2500>
    componentLinearAccScale=<0.6 0.6 0.6>
    componentLinearVelScale=<0.6 0.6 0.6>
    gravityOverride {z=50}
    physicsBodyDefinitions [{
      boundBone {boneName=dyn_strap_e_03}
      boxExtents=<10 40 40>
      constraintSetup {angularLimitsMax=<0 10 5>, angularLimitsMin=<0 -15 -40>, angularTarget=<1 0 0>}
      localJointOffset=<2 0 0>
     }]
   }
   n40 = Anim Dynamics - Bone: dyn_strap_e_04(ComponentPose=@n39, Alpha=1, GravityOverride="0,0,50", ExternalForce="0,0,0") {
    angularDampingOverride=1
    angularSpringConstant=10000
    bAngularSpring=true
    bOverrideAngularDamping=true
    bUseGravityOverride=true
    gravityScale=0.35
    simulationSpace=RootRelative
    alphaBoolBlend {blendOption=Linear}
    boundBone {boneName=dyn_strap_e_04}
    componentAppliedLinearAccClamp=<2000 2000 2500>
    componentLinearAccScale=<0.6 0.6 0.6>
    componentLinearVelScale=<0.6 0.6 0.6>
    gravityOverride {z=50}
    physicsBodyDefinitions [{
      boundBone {boneName=dyn_strap_e_04}
      boxExtents=<10 40 40>
      constraintSetup {angularLimitsMax=<0 10 5>, angularLimitsMin=<0 -15 -40>, angularTarget=<1 0 0>}
      localJointOffset=<4 0 0>
     }]
   }
   Output Pose(Result=Blend Poses by bool(BlendPose_0=Component To Local(ComponentPose=@n20), BlendPose_1=Component To Local(ComponentPose=@n40), bActiveValue=Get bIsInFrontEnd()) {blendType=Linear}) {layerGroup=None}
   n41 = "Transform (Modify) Bone - Bone: pack"(ComponentPose=Local To Component(LocalPose=Input Pose()), Translation="0,-1,0", Rotation="0,0,0") {
    translationMode=BMM_Additive
    translationSpace=BCS_BoneSpace
    alphaBoolBlend {blendOption=Linear}
    boneToModify {boneName=pack}
    translation {y=-1}
   }
   n42 = "Transform (Modify) Bone - Bone: pack"!(ComponentPose=@n41, bAlphaBoolEnabled:bool, Translation="0,1.5,-5", Rotation="0,0,0", Scale) -> (Pose) {
    alphaInputType=Bool
    rotationMode=BMM_Additive
    translationMode=BMM_Additive
    translationSpace=BCS_BoneSpace
    alphaBoolBlend {blendOption=Linear}
    boneToModify {boneName=pack}
    translation {y=1.5, z=-5}
   }
   n43 = Anim Dynamics - Bone: dyn_zipper_01!(ComponentPose=@n42, Alpha=1, GravityOverride:Vector="0,0,-980", ExternalForce="0,0,0") -> (Pose) {
    angularSpringConstant=4.5
    bAngularSpring=true
    bUseGravityOverride=true
    bUsePlanarLimit=false
    alphaBoolBlend {blendOption=Linear}
    boundBone {boneName=dyn_zipper_01}
    componentAppliedLinearAccClamp=<2000 2000 2000>
    componentLinearAccScale=<0.3 0.3 0.3>
    componentLinearVelScale=<0.3 0.3 0.3>
    gravityOverride {z=-980}
    physicsBodyDefinitions [{
      boundBone {boneName=dyn_zipper_01}
      boxExtents=<10 12 12>
      constraintSetup {angularLimitsMax=<0 90 25>, angularLimitsMin=<0 0 -25>, angularTarget=<0 0 0>}
      localJointOffset=<2 0 0>
     }]
   }
   n44 = Look At - Bone: pack!(ComponentPose=@n43, Alpha=1, LookAtLocation:Vector=Get PelvisLocation() -> (PelvisLocation:Vector)) -> (Pose) {
    alphaBoolBlend {blendOption=Linear}
    boneToModify {boneName=pack}
    lookAt_Axis {axis {y=0.7, z=-1}}
    lookUp_Axis {axis {x=0, z=1}}
    lookAtTarget=~
   }
   n45 = "Transform (Modify) Bone - Bone: pack"(ComponentPose=@n44, Translation="0,-1,0", Rotation="0,0,15") {
    alphaInputType=Bool
    rotationMode=BMM_Additive
    rotationSpace=BCS_BoneSpace
    translationMode=BMM_Additive
    translationSpace=BCS_BoneSpace
    alphaBoolBlend {blendOption=Linear}
    boneToModify {boneName=pack}
    rotation {roll=15}
    translation {y=-1}
   }
   Save cached pose 'DefaultPose'(Pose=Component To Local(ComponentPose=@n45)) {cacheName=DefaultPose}
  ]}
 EventGraph {nodes [
   n1 = Event OnCharacterCustomizationLoaded()
   > n2 = Sequence()
   > n3 = Cast To FortPlayerPawnAthena(Object=TryGetPawnOwner())
   > Set bodyMesh(bodyMesh:Skeletal Mesh Component Object Reference=GetSkeletalMeshForPartType(self=@n3)) -> (Output_Get:Skeletal Mesh Component Object Reference)
   Set bodyMesh(execute=@n3.CastFailed, bodyMesh=@n1.CustomBaseMesh)
   n4 = Get Array Item with EFort Custom Part Type(Array:Array of Custom Character Part Object References=@n1.CustomParts) -> (Output:Custom Character Part Object Reference)
   Set bIsOnLargeBodyType(
    execute=@n2.then_1
    bIsOnLargeBodyType:bool="Equal (Enum)"(
     A:EFortCustomBodyType Enum=Get BodyTypesPermitted!(self:Custom Character Part Object Reference=@n4) -> (BodyTypesPermitted:EFortCustomBodyType Enum=NONE)
     B:EFortCustomBodyType Enum=Large
    )
   ) -> (Output_Get:bool)
   > Set harptar_body(harptar_body:bool="Equal (Object)"(A=@n4, B=/BRCosmetics/Athena/Heroes/Meshes/Bodies/CP_Athena_Body_F_HarpTar)) -> (Output_Get:bool)
   Event BlueprintUpdateAnimation()
   n5 = Get bodyMesh() -> (bodyMesh:Skeletal Mesh Component Object Reference)
   > Is Valid(InputObject=@n5)
   > Set PelvisLocation(PelvisLocation:Vector=GetSocketLocation!(self:Scene Component Object Reference=@n5, InSocketName:Name=pelvis) -> (ReturnValue:Vector)) -> (Output_Get:Vector)
   > n6 = Sequence!(execute) -> (then_0, then_1, then_2:Exec)
   > n7 = Branch(Condition=Get bIsSkyDiving())
   > Set Gravity(Gravity:Vector="0,0,100") -> (Output_Get:Vector)
   > Set ExternalForce(ExternalForce:Vector=vector * vector(A=Get RandomVector() -> (RandomVector:Vector), B:float=200) -> (ReturnValue=~)) -> (Output_Get:Vector)
   Set Gravity(execute=@n7.else, Gravity="0,0,-980")
   > Set ExternalForce(ExternalForce="0, 0, 0")
   Delay(execute=@n6.then_1, Duration=0.5)
   > Set RandomVector(RandomVector:Vector=RandomUnitVector!() -> (ReturnValue:Vector)) -> (Output_Get:Vector)
   TryGetPawnOwner()
   K2Node_Event_0 = CustomEvent!() -> (
    OutputDelegate
    then
    CustomBaseMesh:Skeletal Mesh Component Object Reference
    CustomParts:Array of Custom Character Part Object References
    CustomMeshParts:Array of Skeletal Mesh Component Budgeted Object References
   ) {is=title K2Node_CustomEvent}
  ]}
]
variables [
 bIsOnLargeBodyType {type=boolean}
 bodyMesh {type=SkeletalMeshComponent}
 ExternalForce {type=Vector}
 Gravity {type=Vector}
 harptar_body {type=boolean}
 PelvisLocation {type=Vector}
 RandomVector {type=Vector}
]
